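; Comet crater FX actor that shows a glow sphere and atmosphere particles, with a variant for a large crater.
; The construction script scales the sphere and feeds a dynamic material its fade and opacity settings, deriving fade offset from scale.
; The LargeCrater flag overrides glow colors, swaps in the BigOne particle template and starts looping audio at BeginPlay.
; Tick and overlap events are empty.
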(event EventBeginPlay
  (if (Variables|Default|GetLargeCrater)
    (Audio|Components|Audio|Play (Variables|Default|GetAudioLoop) 0.0)))

(event Collision|EventActorBeginOverlap
  (params (OtherActor :type Actor%20Object%20Reference)))

(event EventTick
  (params (DeltaSeconds :type Float%20%28single-precision%29)))

(fn ConstructionScript
  (bind _glowsphere (Variables|Default|GetGlowSphere))
  (bind _glowspherescale (Variables|Default|GetGlowSphereScale))
  (Transformation|SetRelativeScale3D _glowsphere (Math|Conversions|ToVector(Float) _glowspherescale))
  (bind _returnvalue (Rendering|Material|CreateDynamicMaterialInstance _glowsphere 0 0 "None" :!class /Script/Engine.PrimitiveComponent))
  (Rendering|Material|SetScalarParameterValue _returnvalue "FadeDistanceOffset" (* _glowspherescale 6e+01) :!class /Script/Engine.MaterialInstanceDynamic)
  (Rendering|Material|SetScalarParameterValue _returnvalue "FadeDistance" (Variables|Default|GetGlowSphereFadeDistance) :!class /Script/Engine.MaterialInstanceDynamic)
  (Rendering|Material|SetScalarParameterValue _returnvalue "OpacityMult" (Variables|Default|GetGlowSphereOpacityMult) :!class /Script/Engine.MaterialInstanceDynamic)
  (if (Variables|Default|GetLargeCrater)
    (Rendering|Material|SetVectorParameterValue _returnvalue "GlowColor" "(R=0.989677,G=1.000000,B=0.897408,A=1.000000)" :!class /Script/Engine.MaterialInstanceDynamic)
    (Rendering|Material|SetVectorParameterValue _returnvalue "GlowColorOuter" "(R=0.475749,G=0.578125,B=0.530226,A=1.000000)" :!class /Script/Engine.MaterialInstanceDynamic)
    (Effects|Components|ParticleSystem|SetTemplate (Variables|Default|GetP_Comet_Crater_Atmosphere) "/Figment_S04/Athena/Events/P_Comet_Crater_Atmosphere_BigOne.P_Comet_Crater_Atmosphere_BigOne")))
