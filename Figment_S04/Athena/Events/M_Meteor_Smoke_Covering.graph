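# Translucent lit mesh-particle material for the smoke covering on a meteor event effect, previewed on SM_Meteor_Smoke_Covering.
# A smooth noise texture distorts two differently tiled, panning inky smoke samples on the second UV channel, which blend into one mask.
# Opacity multiplies that mask by a vertical UV gradient, vertex red, view-angle fade and depth fade.
# Color lerps from dark grey to warm orange by smoke plus Fresnel, then is dimmed.
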Material /Figment_S04/Athena/Events/M_Meteor_Smoke_Covering
graph {
 properties {
  bUsedWithMeshParticles=true
  blendMode=BLEND_Translucent
  pixelDepthOffsetMode=PDOM_Legacy
  previewMesh=&/SM_Meteor_Smoke_Covering
  usedShadingModels=Default Lit
 }
 nodes [
  n1 = TextureCoordinate() {uTiling=2, vTiling=2}
  n2 = Multiply(A=Time(), B=ScalarParameter() {defaultValue=1.6, parameterName=PanSpeed})
  n3 = Multiply(
   A=TextureSample(UVs=Panner(Coordinate=@n1, Time=@n2) {speedX=0.03, speedY=0.05}) {texture=/Game/Effects/Fort_Effects/Textures/T_Smooth_tile}.R
   B=Constant2Vector() {g=0.07, r=0.07}
  )
  n4 = Add(A=TextureCoordinate() {coordinateIndex=1, uTiling=6, vTiling=0.7}, B=@n3)
  n5 = Add(A=@n3, B=TextureCoordinate() {coordinateIndex=1, uTiling=8, vTiling=1.5})
  n6 = LinearInterpolate(
   A=TextureSample(UVs=Panner(Coordinate=@n4, Time=@n2) {speedX=0.006, speedY=-0.05}) {texture=/Game/Effects/Fort_Effects/Textures/Tile/T_Inky_Smoke_Tile}.R
   B=TextureSample(UVs=Panner(Coordinate=@n5, Time=@n2) {speedX=-0.01, speedY=-0.05}) {texture=&T_Inky_Smoke_Tile}.R
  ) {constAlpha=0.4}
  n7 = OneMinus(None=ComponentMask(None=TextureCoordinate() {coordinateIndex=1}) {g=true})
  n8 = Multiply(
   A=VertexColor().R
   B=MaterialFunctionCall("FadeSensitivity (optional)"=ScalarParameter() {defaultValue=1.5, parameterName=AngleFade}) {materialFunction=&FadeBasedOnViewAngle}.Result
  )
  Multiply(
   A=Multiply(A=@n6, B=ScalarParameter() {defaultValue=2, parameterName=OpacityMult}) {constB=2}
   B=DepthFade(Opacity=Multiply(A=@n7, B=@n8), FadeDistance=ScalarParameter() {defaultValue=70, parameterName=DepthFadeDist})
  )
  n9 = Add(A=Fresnel() {baseReflectFraction=0, exponent=3}, B=@n6)
  Multiply(A=LinearInterpolate(
    A=VectorParameter() {parameterName=DarkColor, defaultValue=<rgba 0.0592 0.0561 0.0677 1>}.RGB
    B=VectorParameter() {parameterName=LightColor, defaultValue=<rgba 1 0.491 0.276 1>}.RGB
    Alpha=Multiply(A=@n9) {constB=0.8}
   )) {constB=0.1}
 ]
}
